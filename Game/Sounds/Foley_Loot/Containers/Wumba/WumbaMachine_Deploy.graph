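# Sound cue for deploying the Wumba machine, a loot-container foley effect.
# A mixer layers the Wumba deploy and placed waves at the same time, each through its own modulator left at default settings.
# It uses item-interact attenuation and the loot reward sound class, and plays when silent rather than stopping.
# A concurrency override caps simultaneous instances at two, stopping the oldest.
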SoundCue /Game/Sounds/Foley_Loot/Containers/Wumba/WumbaMachine_Deploy
root=Mixer_0
nodes [
 {id=Mixer_0, children [Modulator_0,Modulator_1], properties {inputVolume [1,1]}}
 {id=Modulator_0, children [WavePlayer_4]}
 {id=Modulator_1, children [WavePlayer_0]}
 {id=WavePlayer_0, properties {soundWaveAssetPtr=&/Wumba_Placed}}
 {id=WavePlayer_4, properties {soundWaveAssetPtr=&/Wumba_Deploy}}
]
properties {
 attenuationSettings=/Game/Sounds/Attenuation/Fort_Item_Interact_Att
 bOverrideConcurrency=true
 maxDistance=1850
 soundClassObject=&Fort_Loot_Reward_PSM
 virtualizationMode=PlayWhenSilent
 volumeMultiplier=1.1
 concurrencyOverrides {maxCount=2, resolutionRule=StopOldest, volumeScale=0.75}
}
