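# Sound Cue for searching the Wumba machine container, routed to the loot collect sound class with item interact attenuation.
# A Mixer layers a quiet one-shot search wave, passed through a Modulator for random variation, under a much louder looping search wave.
# A concurrency override allows only one instance, stopping the oldest when a new one starts.
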SoundCue /Game/Sounds/Foley_Loot/Containers/Wumba/WumbaMachine_Search_Cue
root=Mixer_0
nodes [
 {id=Mixer_0, children [Modulator_0,WavePlayer_0], properties {inputVolume [0.15,1.25]}}
 {id=Modulator_0, children [WavePlayer_1]}
 {id=WavePlayer_0, properties {bLooping=true, soundWaveAssetPtr=&/Wumba_Search_Loop}}
 {id=WavePlayer_1, properties {soundWaveAssetPtr=&/Wumba_Search_01}}
]
properties {
 attenuationSettings=/Game/Sounds/Attenuation/Fort_Item_Interact_Att
 bOverrideConcurrency=true
 maxDistance=1850
 soundClassObject=&Fort_Loot_Collect
 volumeMultiplier=1.5
 concurrencyOverrides {maxCount=1, resolutionRule=StopOldest}
}
