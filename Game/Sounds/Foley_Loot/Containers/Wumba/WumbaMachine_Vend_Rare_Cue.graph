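# Sound cue for a Wumba machine vending a rare reward, judging by its name and waves.
# A mixer layers the rare vend sound with a slightly quieter rattle layer.
# That layer picks one of three impact rattle variants at equal weight and passes it through a modulator left at its usual settings.
# The cue plays slightly boosted in the loot reward sound class with item-interact attenuation.
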SoundCue /Game/Sounds/Foley_Loot/Containers/Wumba/WumbaMachine_Vend_Rare_Cue
root=Mixer_0
nodes [
 {id=Mixer_0, children [WavePlayer_0,Modulator_0], properties {inputVolume [1,0.75]}}
 {id=Modulator_0, children [Random_0]}
 {id=Random_0, children [WavePlayer_1,WavePlayer_2,WavePlayer_3], properties {weights [1,1,1]}}
 {id=WavePlayer_0, properties {soundWaveAssetPtr=&/Wumba_Rare_Vend}}
 {id=WavePlayer_1, properties {soundWaveAssetPtr=&/Wumba_Impact_Rattle}}
 {id=WavePlayer_2, properties {soundWaveAssetPtr=&/Wumba_Impact_Rattle_02}}
 {id=WavePlayer_3, properties {soundWaveAssetPtr=&/Wumba_Impact_Rattle_03}}
]
properties {
 attenuationSettings=/Game/Sounds/Attenuation/Fort_Item_Interact_Att
 maxDistance=1850
 soundClassObject=&Fort_Loot_Reward_PSM
 volumeMultiplier=1.4
}
