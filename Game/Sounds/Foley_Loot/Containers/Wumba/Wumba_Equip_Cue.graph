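# Sound cue that plays the Wumba equip foley, filed under loot container sounds.
# A single wave player feeds a modulator that applies a slight random upward pitch variation.
# The cue plays at reduced volume in the gadgets and items sound class, using weapon-drop attenuation for spatial falloff.
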SoundCue /Game/Sounds/Foley_Loot/Containers/Wumba/Wumba_Equip_Cue
root=Modulator_0
nodes [
 {id=Modulator_0, children [WavePlayer_0], properties {pitchMax=1.1, pitchMin=1}}
 {id=WavePlayer_0, properties {soundWaveAssetPtr=&/Wumba_Equip}}
]
properties {
 attenuationSettings=&Fort_Weapon_Drop_Att
 maxDistance=3000
 soundClassObject=&Fort_Gadgets_Items
 volumeMultiplier=0.3
}
