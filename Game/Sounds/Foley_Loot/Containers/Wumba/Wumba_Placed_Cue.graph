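# Sound cue for placing the Wumba container, filed under loot container foley.
# It plays the single Wumba_Placed wave through a Modulator that randomizes pitch slightly upward on each play.
# Output is turned down and routed to the gadget items sound class, with spatial falloff from a shared trap-launch attenuation asset.
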SoundCue /Game/Sounds/Foley_Loot/Containers/Wumba/Wumba_Placed_Cue
root=Modulator_0
nodes [
 {id=Modulator_0, children [WavePlayer_1], properties {pitchMax=1.1, pitchMin=1}}
 {id=WavePlayer_1, properties {soundWaveAssetPtr=&/Wumba_Placed}}
]
properties {
 attenuationSettings=/Game/Sounds/Attenuation/Fort_Traps_Launch_Athena_Att
 maxDistance=7000
 soundClassObject=&Fort_Gadgets_Items
 volumeMultiplier=0.35
}
